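# Idle Niagara effect for the ToothGreek pickaxe cosmetic that draws three translucent blue capsule glow sprites attached to the weapon.
# A single burst is split into three partitions, each picking its own sprite size and following the socket of the same index through a socket reader.
# The glows do not expire by lifetime, and their color is scaled by system-level fade and time-of-day color values.
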NiagaraSystem /BRCosmetics/Weapons/FORT_Melee/Pickaxe_ToothGreek/FX/NS_Pickaxe_ToothGreek_Idle
dependencies {
 usedDataInterfaces [&NiagaraDataInterfaceSocketReader, &NiagaraDataInterfaceCurve]
 usedDynamicInputs [
  &MergeHelper_Vector2Switch
  &FloatFromCurve
  &FNTimeOfDayColor
 ]
 usedModules [
  &self:SystemSpawnScript.SystemScriptSource.SystemScriptGraph.NiagaraNodeAssignment_0.SetVariables_#1
  &SystemState
  &self:SystemSpawnScript.SystemScriptSource.SystemScriptGraph.NiagaraNodeAssignment_2.SetVariables_#2
  &self:SystemSpawnScript.SystemScriptSource.SystemScriptGraph.NiagaraNodeAssignment_1.SetVariables_#3
 ]
}
emitters {CapsuleGlows {
  emitterUpdateScript [
   &EmitterState(Life Cycle Mode=::NewEnumerator1, Loop Behavior=::NewEnumerator1, Loop Duration Mode=::NewEnumerator1, UseLoopDelay=false)
   &SpawnBurst_Instantaneous(Spawn Count=3, Use Spawn Probability=false, UseLoopCountLimit=true)
  ]
  particleSpawnScript [
   &PartitionParticles(ID to Partition:NiagaraInt32=&ReturnExecIndex(), Number Of Partitions=3)
   &InitializeParticle(
    Color=<rgba 0 0.3502 1 0.5>
    Position Mode=::NewEnumerator0
    Sprite Size Mode=::NewEnumerator3
    Uniform Sprite Size=&Multiply_Float(
     A=12
     B=&MergeHelper_FloatSwitch(0=1, 1=0.57, 2=0.38, 3=?, Index=$Particles.PartitionParticles.Partition, Num=::NewEnumerator1)
    )
    Write Material Random=true
   )
  ]
  particleUpdateScript [
   &SocketLocation(
    SamplePosOffset=false
    Socket Index=$Particles.PartitionParticles.Partition
    Socket Reader=$System.Socket Reader
    Socket Sampling Mode=::NewEnumerator3
    VelocitySpeedLimit=false
   )
   &ParticleState(Kill Particles When Lifetime Has Elapsed=false, Let Infinitely Lived Particles Die When Emitter Deactivates=true)
   &DynamicMaterialParameters(
    Index 0 Param 1=1
    Index 0 Param 2=10
    Module.Index0Param0Write=true
    Module.Index0Param1Write=false
    Module.Index0Param2Write=false
    Module.Index0Param3Write=false
   ) {module=DynamicMaterialParameters001}
   &ScaleColor(Scale Alpha=$System.SystemFade, ScaleA=true, ScaleRGB=true)
   &ScaleColor(Color Scale:LinearColor=$System.TODColor, Scale Mode=3) {module=ScaleColor001}
  ]
  rendererClasses [&NiagaraSpriteRendererProperties]
  renderers [{rendererClass=&NiagaraSpriteRendererProperties}]
 }}
summary {
 systemName=NS_Pickaxe_ToothGreek_Idle
 emitters [{emitterName=CapsuleGlows, rendererClasses [&NiagaraSpriteRendererProperties]}]
 userVariables [
  User.bIsFrontEnd {type=NiagaraBool}
  User.bIsFrontEndPreview {type=NiagaraBool}
 ]
}
