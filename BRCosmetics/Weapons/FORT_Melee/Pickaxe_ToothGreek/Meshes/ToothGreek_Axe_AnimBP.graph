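# Animation Blueprint for the ToothGreek pickaxe that adds secondary physics motion to its left and right bow bones.
# The cached input pose feeds three Anim Dynamics chains with angular springs, limits and custom gravity.
# Outside the front end, a full-strength chain plays; in it, bIsOnPawn selects a reduced-alpha chain, otherwise a mostly world-space chain skipping dyn_bow2_l.
# A fourth chain is unconnected, and the event graph does nothing.
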AnimBlueprint /BRCosmetics/Weapons/FORT_Melee/Pickaxe_ToothGreek/Meshes/ToothGreek_Axe_AnimBP
target_skeleton=&/ToothGreek_Axe_Skeleton
graphs [
 AnimGraph {nodes [
   n1 = Anim Dynamics - Bone: dyn_bow1_r(Alpha=1, ExternalForce="0,0,0") {
    angularDampingOverride=0.9998
    angularSpringConstant=900
    bAngularSpring=true
    bOverrideAngularDamping=true
    bUseGravityOverride=true
    linearSpringConstant=500
    boundBone {boneName=dyn_bow1_r}
    componentAppliedLinearAccClamp=<10000 10000 10000>
    componentLinearAccScale=<0.4 1 0.4>
    componentLinearVelScale=<0.2 0.2 0.2>
    gravityOverride {z=-400}
    physicsBodyDefinitions [{
      boundBone {boneName=dyn_bow1_r}
      boxExtents=<10 10 10>
      constraintSetup {angularLimitsMax=<30 60 5>, angularLimitsMin=<-30 -10 5>, angularTarget=<1 0 0>}
      localJointOffset=<1.6 0 0>
     }]
   }
   n2 = Anim Dynamics - Bone: dyn_bow2_r(ComponentPose=@n1, Alpha=1, ExternalForce="0,0,0") {
    angularDampingOverride=1
    angularSpringConstant=700
    bAngularSpring=true
    bOverrideAngularDamping=true
    bUseGravityOverride=true
    boundBone {boneName=dyn_bow2_r}
    componentAppliedLinearAccClamp=<10000 10000 10000>
    componentLinearAccScale=<0.4 0.4 0.4>
    componentLinearVelScale=<0.2 0.2 0.2>
    gravityOverride {z=-1000}
    physicsBodyDefinitions [{
      boundBone {boneName=dyn_bow2_r}
      boxExtents=<10 10 10>
      constraintSetup {angularLimitsMax=<30 10 0>, angularLimitsMin=<-30 -15 0>, angularTarget=<1 0 0>}
      localJointOffset=<0.8 0 0>
     }]
   }
   n3 = Anim Dynamics - Bone: dyn_bow1_l(ComponentPose=@n2, Alpha=1, ExternalForce="0,0,0") {
    angularDampingOverride=0.9998
    angularSpringConstant=1300
    bAngularSpring=true
    bOverrideAngularDamping=true
    bUseGravityOverride=true
    boundBone {boneName=dyn_bow1_l}
    componentAppliedLinearAccClamp=<10000 10000 10000>
    componentLinearAccScale=<0.4 1 0.4>
    componentLinearVelScale=<0.2 0.6 0.2>
    gravityOverride {z=-500}
    physicsBodyDefinitions [{
      boundBone {boneName=dyn_bow1_l}
      boxExtents=<10 10 10>
      constraintSetup {angularLimitsMax=<10 40 60>, angularLimitsMin=<-10 -50 -10>, angularTarget=<1 0 0>}
      localJointOffset=<1.4 0 0>
     }]
   }
   Anim Dynamics - Bone: dyn_bow2_l(ComponentPose=@n3, Alpha=1, ExternalForce="0,0,0") {
    angularDampingOverride=1
    angularSpringConstant=1100
    bAngularSpring=true
    bOverrideAngularDamping=true
    bUseGravityOverride=true
    linearSpringConstant=1000
    boundBone {boneName=dyn_bow2_l}
    componentAppliedLinearAccClamp=<10000 10000 10000>
    componentLinearAccScale=<0.4 0.4 0.4>
    componentLinearVelScale=<0.2 0.2 0.2>
    gravityOverride {z=-500}
    physicsBodyDefinitions [{
      boundBone {boneName=dyn_bow2_l}
      boxExtents=<7 7 7>
      constraintSetup {angularLimitsMax=<10 15 0>, angularLimitsMin=<-10 -10 0>, angularTarget=<1 0 0>}
      localJointOffset=<0.9 0 0>
     }]
   }
   n4 = Anim Dynamics - Bone: dyn_bow1_r(ComponentPose=Local To Component(LocalPose=Use cached pose 'baseCache'()), Alpha=1, ExternalForce="0,0,0") {
    angularDampingOverride=0.98
    angularSpringConstant=300
    bAngularSpring=true
    bOverrideAngularDamping=true
    bUseGravityOverride=true
    simulationSpace=World
    boundBone {boneName=dyn_bow1_r}
    componentAppliedLinearAccClamp=<10000 10000 10000>
    componentLinearAccScale=<0.4 0.4 0.4>
    componentLinearVelScale=<0.2 0.2 0.2>
    gravityOverride {z=-900}
    physicsBodyDefinitions [{
      boundBone {boneName=dyn_bow1_r}
      boxExtents=<10 10 10>
      constraintSetup {angularLimitsMax=<0 40 0>, angularLimitsMin=<0 -40 0>, angularTarget=<1 0 0>}
      localJointOffset=<1 0 0>
     }]
   }
   n5 = Anim Dynamics - Bone: dyn_bow2_r(ComponentPose=@n4, Alpha=1, ExternalForce="0,0,0") {
    angularDampingOverride=1
    angularSpringConstant=800
    bAngularSpring=true
    bOverrideAngularDamping=true
    bUseGravityOverride=true
    simulationSpace=World
    boundBone {boneName=dyn_bow2_r}
    componentAppliedLinearAccClamp=<10000 10000 10000>
    componentLinearAccScale=<0.4 0.4 0.4>
    componentLinearVelScale=<0.2 0.2 0.2>
    gravityOverride {z=-1100}
    physicsBodyDefinitions [{
      boundBone {boneName=dyn_bow2_r}
      boxExtents=<10 10 10>
      constraintSetup {angularLimitsMax=<0 12 0>, angularLimitsMin=<0 -20 0>, angularTarget=<1 0 0>}
      localJointOffset=<1.5 0 0>
     }]
   }
   n6 = Anim Dynamics - Bone: dyn_bow3_r(ComponentPose=@n5, Alpha=1, ExternalForce="0,0,0") {
    angularDampingOverride=1
    angularSpringConstant=2000
    bAngularSpring=true
    bOverrideAngularDamping=true
    bUseGravityOverride=true
    simulationSpace=World
    boundBone {boneName=dyn_bow3_r}
    componentAppliedLinearAccClamp=<10000 10000 10000>
    componentLinearAccScale=<0.4 0.4 0.4>
    componentLinearVelScale=<0.2 0.2 0.2>
    gravityOverride {z=-400}
    physicsBodyDefinitions [{
      boundBone {boneName=dyn_bow3_r}
      boxExtents=<12 12 12>
      constraintSetup {angularLimitsMax=<30 15 0>, angularLimitsMin=<-30 -15 0>, angularTarget=<1 0 0>}
      localJointOffset=<0.4 0 0>
     }]
   }
   n7 = Anim Dynamics - Bone: dyn_bow1_l(ComponentPose=@n6, Alpha=1, ExternalForce="0,0,0") {
    angularDampingOverride=0.98
    angularSpringConstant=300
    bAngularSpring=true
    bOverrideAngularDamping=true
    bUseGravityOverride=true
    simulationSpace=World
    boundBone {boneName=dyn_bow1_l}
    componentAppliedLinearAccClamp=<10000 10000 10000>
    componentLinearAccScale {y=0.4}
    componentLinearVelScale=<0.2 0.2 0.2>
    gravityOverride {z=-900}
    physicsBodyDefinitions [{
      boundBone {boneName=dyn_bow1_l}
      boxExtents=<20 20 20>
      constraintSetup {angularLimitsMax=<0 40 0>, angularLimitsMin=<0 -40 0>, angularTarget=<1 0 0>}
      localJointOffset=<1 0 0>
     }]
   }
   n8 = Anim Dynamics - Bone: dyn_bow3_l(ComponentPose=@n7, Alpha=1, ExternalForce="0,0,0") {
    angularDampingOverride=1
    angularSpringConstant=1500
    bAngularSpring=true
    bOverrideAngularDamping=true
    bUseGravityOverride=true
    boundBone {boneName=dyn_bow3_l}
    componentAppliedLinearAccClamp=<10000 10000 10000>
    componentLinearAccScale=<0.4 0.4 0.4>
    componentLinearVelScale=<0.2 0.2 0.2>
    gravityOverride {z=-200}
    physicsBodyDefinitions [{
      boundBone {boneName=dyn_bow3_l}
      boxExtents=<12 12 12>
      constraintSetup {angularLimitsMax=<0 10 0>, angularLimitsMin=<0 -10 0>, angularTarget=<1 0 0>}
      localJointOffset=<0.6 0 0>
     }]
   }
   n9 = Component To Local(ComponentPose=@n8)
   n10 = Anim Dynamics - Bone: dyn_bow1_r(ComponentPose=Local To Component(LocalPose=Use cached pose 'baseCache'()), Alpha=0.8, ExternalForce="0,0,0") {
    angularDampingOverride=1
    angularSpringConstant=2000
    bAngularSpring=true
    bOverrideAngularDamping=true
    bUseGravityOverride=true
    linearSpringConstant=500
    boundBone {boneName=dyn_bow1_r}
    componentAppliedLinearAccClamp=<10000 10000 10000>
    componentLinearAccScale=<0.4 1 0.4>
    componentLinearVelScale=<0.2 0.2 0.2>
    gravityOverride {z=-400}
    physicsBodyDefinitions [{
      boundBone {boneName=dyn_bow1_r}
      boxExtents=<10 10 10>
      constraintSetup {angularLimitsMax=<30 60 0>, angularLimitsMin=<-30 -60 0>, angularTarget=<1 0 0>}
      localJointOffset=<0.6 0 0>
     }]
   }
   n11 = Anim Dynamics - Bone: dyn_bow2_r(ComponentPose=@n10, Alpha=0.8, ExternalForce="0,0,0") {
    angularDampingOverride=1
    angularSpringConstant=2000
    bAngularSpring=true
    bOverrideAngularDamping=true
    bUseGravityOverride=true
    boundBone {boneName=dyn_bow2_r}
    componentAppliedLinearAccClamp=<10000 10000 10000>
    componentLinearAccScale=<0.4 0.4 0.4>
    componentLinearVelScale=<0.2 0.2 0.2>
    gravityOverride {z=-400}
    physicsBodyDefinitions [{
      boundBone {boneName=dyn_bow2_r}
      boxExtents=<12 12 12>
      constraintSetup {angularLimitsMax=<30 20 0>, angularLimitsMin=<-30 -20 0>, angularTarget=<1 0 0>}
      localJointOffset=<0.6 0 0>
     }]
   }
   n12 = Anim Dynamics - Bone: dyn_bow3_r(ComponentPose=@n11, Alpha=0.8, ExternalForce="0,0,0") {
    angularDampingOverride=1
    angularSpringConstant=2000
    bAngularSpring=true
    bOverrideAngularDamping=true
    bUseGravityOverride=true
    boundBone {boneName=dyn_bow3_r}
    componentAppliedLinearAccClamp=<10000 10000 10000>
    componentLinearAccScale=<0.4 0.4 0.4>
    componentLinearVelScale=<0.2 0.2 0.2>
    gravityOverride {z=-400}
    physicsBodyDefinitions [{
      boundBone {boneName=dyn_bow3_r}
      boxExtents=<12 12 12>
      constraintSetup {angularLimitsMax=<30 15 0>, angularLimitsMin=<-30 -15 0>, angularTarget=<1 0 0>}
      localJointOffset=<0.4 0 0>
     }]
   }
   n13 = Anim Dynamics - Bone: dyn_bow1_l(ComponentPose=@n12, Alpha=0.7, ExternalForce="0,0,0") {
    angularDampingOverride=1
    angularSpringConstant=500
    bAngularSpring=true
    bOverrideAngularDamping=true
    bUseGravityOverride=true
    linearSpringConstant=500
    boundBone {boneName=dyn_bow1_l}
    componentAppliedLinearAccClamp=<10000 10000 10000>
    componentLinearAccScale=<0.4 1 0.4>
    componentLinearVelScale=<0.2 0.2 0.2>
    gravityOverride {z=-200}
    physicsBodyDefinitions [{
      boundBone {boneName=dyn_bow1_l}
      boxExtents=<8 8 8>
      constraintSetup {angularLimitsMax=<30 80 0>, angularLimitsMin=<-30 -80 0>, angularTarget=<1 0 0>}
      localJointOffset=<0.8 0 0>
     }]
   }
   n14 = Anim Dynamics - Bone: dyn_bow2_l(ComponentPose=@n13, Alpha=0.7, ExternalForce="0,0,0") {
    angularDampingOverride=1
    angularSpringConstant=500
    bAngularSpring=true
    bOverrideAngularDamping=true
    bUseGravityOverride=true
    boundBone {boneName=dyn_bow2_l}
    componentAppliedLinearAccClamp=<10000 10000 10000>
    componentLinearAccScale=<0.4 0.4 0.4>
    componentLinearVelScale=<0.2 0.2 0.2>
    gravityOverride {z=-200}
    physicsBodyDefinitions [{
      boundBone {boneName=dyn_bow2_l}
      boxExtents=<10 10 10>
      constraintSetup {angularLimitsMax=<30 25 0>, angularLimitsMin=<-30 -25 0>, angularTarget=<1 0 0>}
      localJointOffset=<0.6 0 0>
     }]
   }
   n15 = Anim Dynamics - Bone: dyn_bow3_l(ComponentPose=@n14, Alpha=0.7, ExternalForce="0,0,0") {
    angularDampingOverride=1
    angularSpringConstant=500
    bAngularSpring=true
    bOverrideAngularDamping=true
    bUseGravityOverride=true
    boundBone {boneName=dyn_bow3_l}
    componentAppliedLinearAccClamp=<10000 10000 10000>
    componentLinearAccScale=<0.4 0.4 0.4>
    componentLinearVelScale=<0.2 0.2 0.2>
    gravityOverride {z=-200}
    physicsBodyDefinitions [{
      boundBone {boneName=dyn_bow3_l}
      boxExtents=<10 10 10>
      constraintSetup {angularLimitsMax=<30 10 0>, angularLimitsMin=<-30 -10 0>, angularTarget=<1 0 0>}
      localJointOffset=<0.5 0 0>
     }]
   }
   n16 = Component To Local(ComponentPose=@n15)
   n17 = Get bIsOnPawn() -> (bIsOnPawn:bool)
   n18 = Anim Dynamics - Bone: dyn_bow1_r(ComponentPose=Local To Component(LocalPose=Use cached pose 'baseCache'()), Alpha=1, ExternalForce="0,0,0") {
    angularDampingOverride=1
    angularSpringConstant=2000
    bAngularSpring=true
    bOverrideAngularDamping=true
    bUseGravityOverride=true
    linearSpringConstant=500
    boundBone {boneName=dyn_bow1_r}
    componentAppliedLinearAccClamp=<10000 10000 10000>
    componentLinearAccScale=<0.4 1 0.4>
    componentLinearVelScale=<0.2 0.2 0.2>
    gravityOverride {z=-400}
    physicsBodyDefinitions [{
      boundBone {boneName=dyn_bow1_r}
      boxExtents=<10 10 10>
      constraintSetup {angularLimitsMax=<30 60 0>, angularLimitsMin=<-30 -60 0>, angularTarget=<1 0 0>}
      localJointOffset=<0.6 0 0>
     }]
   }
   n19 = Anim Dynamics - Bone: dyn_bow2_r(ComponentPose=@n18, Alpha=1, ExternalForce="0,0,0") {
    angularDampingOverride=1
    angularSpringConstant=2000
    bAngularSpring=true
    bOverrideAngularDamping=true
    bUseGravityOverride=true
    boundBone {boneName=dyn_bow2_r}
    componentAppliedLinearAccClamp=<10000 10000 10000>
    componentLinearAccScale=<0.4 0.4 0.4>
    componentLinearVelScale=<0.2 0.2 0.2>
    gravityOverride {z=-400}
    physicsBodyDefinitions [{
      boundBone {boneName=dyn_bow2_r}
      boxExtents=<12 12 12>
      constraintSetup {angularLimitsMax=<30 20 0>, angularLimitsMin=<-30 -20 0>, angularTarget=<1 0 0>}
      localJointOffset=<0.6 0 0>
     }]
   }
   n20 = Anim Dynamics - Bone: dyn_bow3_r(ComponentPose=@n19, Alpha=1, ExternalForce="0,0,0") {
    angularDampingOverride=1
    angularSpringConstant=2000
    bAngularSpring=true
    bOverrideAngularDamping=true
    bUseGravityOverride=true
    boundBone {boneName=dyn_bow3_r}
    componentAppliedLinearAccClamp=<10000 10000 10000>
    componentLinearAccScale=<0.4 0.4 0.4>
    componentLinearVelScale=<0.2 0.2 0.2>
    gravityOverride {z=-400}
    physicsBodyDefinitions [{
      boundBone {boneName=dyn_bow3_r}
      boxExtents=<12 12 12>
      constraintSetup {angularLimitsMax=<30 15 0>, angularLimitsMin=<-30 -15 0>, angularTarget=<1 0 0>}
      localJointOffset=<0.4 0 0>
     }]
   }
   n21 = Anim Dynamics - Bone: dyn_bow1_l(ComponentPose=@n20, Alpha=1, ExternalForce="0,0,0") {
    angularDampingOverride=1
    angularSpringConstant=1500
    bAngularSpring=true
    bOverrideAngularDamping=true
    bUseGravityOverride=true
    linearSpringConstant=500
    boundBone {boneName=dyn_bow1_l}
    componentAppliedLinearAccClamp=<10000 10000 10000>
    componentLinearAccScale=<0.4 1 0.4>
    componentLinearVelScale=<0.2 0.2 0.2>
    gravityOverride {z=-200}
    physicsBodyDefinitions [{
      boundBone {boneName=dyn_bow1_l}
      boxExtents=<10 10 10>
      constraintSetup {angularLimitsMax=<30 80 0>, angularLimitsMin=<-30 -80 0>, angularTarget=<1 0 0>}
      localJointOffset=<1 0 0>
     }]
   }
   n22 = Anim Dynamics - Bone: dyn_bow2_l(ComponentPose=@n21, Alpha=1, ExternalForce="0,0,0") {
    angularDampingOverride=1
    angularSpringConstant=1300
    bAngularSpring=true
    bOverrideAngularDamping=true
    bUseGravityOverride=true
    boundBone {boneName=dyn_bow2_l}
    componentAppliedLinearAccClamp=<10000 10000 10000>
    componentLinearAccScale=<0.4 0.4 0.4>
    componentLinearVelScale=<0.2 0.2 0.2>
    gravityOverride {z=-200}
    physicsBodyDefinitions [{
      boundBone {boneName=dyn_bow2_l}
      boxExtents=<12 12 12>
      constraintSetup {angularLimitsMax=<30 25 0>, angularLimitsMin=<-30 -25 0>, angularTarget=<1 0 0>}
      localJointOffset=<0.8 0 0>
     }]
   }
   n23 = Anim Dynamics - Bone: dyn_bow3_l(ComponentPose=@n22, Alpha=1, ExternalForce="0,0,0") {
    angularDampingOverride=1
    angularSpringConstant=1500
    bAngularSpring=true
    bOverrideAngularDamping=true
    bUseGravityOverride=true
    boundBone {boneName=dyn_bow3_l}
    componentAppliedLinearAccClamp=<10000 10000 10000>
    componentLinearAccScale=<0.4 0.4 0.4>
    componentLinearVelScale=<0.2 0.2 0.2>
    gravityOverride {z=-200}
    physicsBodyDefinitions [{
      boundBone {boneName=dyn_bow3_l}
      boxExtents=<12 12 12>
      constraintSetup {angularLimitsMax=<30 10 0>, angularLimitsMin=<-30 -10 0>, angularTarget=<1 0 0>}
      localJointOffset=<0.6 0 0>
     }]
   }
   Output Pose(Result=Blend Poses by bool(
     BlendPose_0=Blend Poses by bool(BlendPose_0=@n16, BlendPose_1=@n9, bActiveValue=@n17) {blendType=Linear}
     BlendPose_1=Component To Local(ComponentPose=@n23)
     bActiveValue=Get bIsInFrontEnd()
    ) {blendType=Linear}) {layerGroup=None}
   AnimGraphNode_SubInput_0 = Input Pose()
   Save cached pose 'baseCache'(Pose=@AnimGraphNode_SubInput_0) {cacheName=baseCache}
  ]}
 EventGraph {nodes [
   TryGetPawnOwner()
   Event BlueprintUpdateAnimation()
  ]}
]
variables []
